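# Niagara system for the Ashes backpack cosmetic, adding flames, a fire plume, flickering light, and wing and tail mesh particles.
# Flame and plume sprites follow the user-supplied skeletal mesh, split into two spawn groups with mirrored, curve-driven offsets from sampled sockets or bones.
# Light color and size flicker via loop-age curves, while Wings and Tail pulse a light color with a sine.
# Disabled older flame emitters instead use position lerps weighted by IsSkydiving.
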NiagaraSystem /BRCosmetics/Accessories/FORT_Backpacks/Backpack_F_MED_Ashes/FX/NS_Backpack_Ashes
dependencies {
 usedDataInterfaces [
  &NiagaraDataInterfaceArrayFloat
  &NiagaraDataInterfaceArrayFloat3
  &NiagaraDataInterfaceColorCurve
  &NiagaraDataInterfaceVectorCurve
  &NiagaraDataInterfaceCurve
 ]
 usedDynamicInputs []
 usedModules [&SystemState]
}
emitters {
 Fire_Bottom {
  bEnabled=false
  emitterUpdateScript [
   &EmitterState()
   &SpawnRate(SpawnRate=12, Use Spawn Probability=false)
  ]
  particleSpawnScript [
   /Niagara/Modules/Spawn/Initialization/InitializeParticle(
    Color=<rgba 5 3.0056 1.6206 0.75>
    Lifetime=&UniformRangedFloat(Minimum=0.6, Maximum=0.7)
    Module.Write Color=false
    Module.Write Lifetime=true
    Module.Write Mass=false
    Module.Write Material Random=true
    Module.Write Position=false
    Module.Write Scale=false
    Module.Write SpriteRotation=true
    Module.Write SpriteSize=true
    Position=&SimulationPosition()
    Sprite Rotation=&UniformRangedFloat(Minimum=-20, Maximum=20)
    Sprite Size=&MakeCustomVector2DFromBool(Bool=&RandomBool(), True Vector2D=<52 52>, False Vector2D=<-52 52>)
   )
   &SkeletalMeshLocation(Bone Sampling Mode=::NewEnumerator4, Inherit Velocity=false, MaskSpawnGroup=false, Skeletal Mesh=$User.SkeletalMesh)
   /Niagara/Modules/Spawn/Location/SphereLocation(MaskSpawnGroup=false, Module.Override Randomness=false, Non Uniform Scale=<1 1 0.1>, Sphere Radius=2, Surface Only=false)
   &Color()
   &DynamicMaterialParameters(
    Index 0 Param 2=&UniformRangedFloat(Minimum=0, Maximum=1)
    Index 0 Param 3=1
    Index 0 Param 4=&UniformRangedFloat(Minimum=1, Maximum=2)
    Module.Param0WriteEnabled=false
    Module.Param1WriteEnabled=true
    Module.Param2WriteEnabled=false
    Module.Param3WriteEnabled=true
   ) {module=DynamicMaterialParameters002}
  ]
  particleUpdateScript [
   &ParticleState()
   &SkeletalMeshLocation(
    Bone Sampling Mode=::NewEnumerator4
    Inherit Velocity=false
    Interpolate Mesh Sampling=false
    MaskSpawnGroup=false
    Skeletal Mesh=$User.SkeletalMesh
   ) {module=SkeletalMeshLocation001}
   &self:LERP-Source_to_Destination_01(
    DestinationPosition:NiagaraPosition=$Particles.Initial.Position
    DestinationPosition-Offset:Vector3f=&RandomRangeVector(
     Minimum=&VectorFromCurve(
      VectorCurve=DataInterface() {propertyValues {
        XCurve {keys [<at 0 -4>]}
        YCurve {keys [<at 0 -2>]}
        ZCurve {keys [<at 0 0 RCIM_Cubic>, {arriveTangent=40.1827, interpMode=RCIM_Cubic, leaveTangent=40.183, tangentMode=RCTM_User, time=1, value=27}]}
       }}
      Scale Curve=<1 1 -0.5>
     )
     Maximum=&VectorFromCurve(
      VectorCurve=DataInterface() {propertyValues {
        XCurve {keys [<at 0 4>]}
        YCurve {keys [<at 0 3>]}
        ZCurve {keys [<at 0 0 RCIM_Cubic>, {arriveTangent=36.9034, interpMode=RCIM_Cubic, leaveTangent=36.9036, tangentMode=RCTM_User, time=1, value=29}]}
       }}
      Scale Curve=<1 1 -0.8>
     )
    )
    Normalized Influence:NiagaraFloat=&MakeCustomFloatFromBool(
     Bool=$User.IsSkydiving
     True Float=&FloatFromCurve(FloatCurve=DataInterface() {propertyValues {Curve {keys [
          {arriveTangent=0.0353, interpMode=RCIM_Cubic, leaveTangent=0.0353, tangentMode=RCTM_User, time=0, value=0}
          {arriveTangent=0.0217, interpMode=RCIM_Cubic, leaveTangent=0.0217, tangentMode=RCTM_User, time=1, value=0.025}
         ]}}})
     False Float=&FloatFromCurve(FloatCurve=DataInterface() {propertyValues {Curve {keys [
          {arriveTangent=0.1289, interpMode=RCIM_Cubic, leaveTangent=0.1289, tangentMode=RCTM_User, time=0, value=0}
          {arriveTangent=0.0487, interpMode=RCIM_Cubic, leaveTangent=0.0487, tangentMode=RCTM_User, time=1, value=0.1}
         ]}}})
    )
    SourcePosition:Vector3f=$Output.SkeletalMeshLocation001.SampledPosition
    SourcePosition-Offset:Vector3f=&TransformVector(Vector=&RandomRangeVector(Minimum=<10 -2 -7>, Maximum=<-10 -3 -7>), Destination Space=::Simulation)
   ) {module=LERP-Source_to_Destination_01}
   &ScaleColor(
    Scale Mode=::NewEnumerator1
    Scale RGBA=&Vector4FromCurve(Vector4Curve=DataInterface() {propertyValues {
       AlphaCurve {keys [<at 0.0024 0>,<at 0.15 1.5>,<at 0.45 1>,<at 0.75 2>,<at 1 0>]}
       BlueCurve {keys [<at 0 0>,<at 0.1017 0>,<at 0.6 0>,<at 0.75 0>]}
       GreenCurve {keys [<at 0 0.25>,<at 0.1017 0.1667>,<at 0.6 0.1333>,<at 0.75 0.0833>]}
       RedCurve {keys [<at 0 1>,<at 0.1017 1>,<at 0.6 1>,<at 0.75 1>]}
      }})
    ScaleRGBA=true
   ) {module=ScaleColor002}
   /SetVariables(Particles.SubImageIndex=&Add_Float(
     A=&RandomRangeFloat(Minimum=0, Maximum=20)
     B=&FloatFromCurve(FloatCurve=DataInterface() {propertyValues {Curve {keys [<at 0 0>,<at 1 24>]}}})
    )) {isSetParameters=true, module=SetVariables_#1}
   &ScaleSpriteSize(Scale Factor=&Vector2DFromFloat(Value=&FloatFromCurve(
      FloatCurve=DataInterface() {propertyValues {Curve {keys [
          {arriveTangent=5.8415, interpMode=RCIM_Cubic, leaveTangent=5.8415, tangentMode=RCTM_User, time=0, value=0}
          {arriveTangent=0.4267, interpMode=RCIM_Cubic, leaveTangent=0.4267, tangentMode=RCTM_User, time=0.25, value=1}
          {arriveTangent=-1.3333, interpMode=RCIM_Cubic, leaveTangent=-1.3333, tangentMode=RCTM_User, time=0.5752, value=1}
          {arriveTangent=-3.118, interpMode=RCIM_Cubic, leaveTangent=-3.118, tangentMode=RCTM_User, time=1, value=0.5}
         ]}}}
      Scale Curve=0.5
     )))
   &DynamicMaterialParameters(
    Index 0 Param 1=&FloatFromCurve(FloatCurve=DataInterface() {propertyValues {Curve {keys [{arriveTangent=2.7551, interpMode=RCIM_Cubic, leaveTangent=2.7551, tangentMode=RCTM_User, time=0, value=0}, <at 0.5 1>]}}})
    Index 0 Param 2=&RandomRangeFloat(Minimum=0.33, Maximum=1)
    Index 0 Param 3=&FloatFromCurve(FloatCurve=DataInterface() {propertyValues {Curve {keys [<at 0 1>,<at 1 2>]}}})
    Index 0 Param 4=1
    Module.Param0WriteEnabled=true
    Module.Param1WriteEnabled=true
    Module.Param2WriteEnabled=true
    Module.Param3WriteEnabled=true
   ) {module=DynamicMaterialParameters001}
   &CameraOffset() {module=CameraOffset001}
   &SpriteRotationRate(Rotation Rate=&RandomRangeFloat(Minimum=-50, Maximum=50))
  ]
  rendererClasses [&NiagaraSpriteRendererProperties]
  renderers [{rendererClass=&NiagaraSpriteRendererProperties}]
 }
 Fire_Bottom001 {
  emitterUpdateScript [
   &EmitterState()
   &SpawnRate(SpawnRate=12, Use Spawn Probability=false)
  ]
  particleSpawnScript [
   &InitializeParticle(
    Color=<rgba 5 3.0056 1.6206 0.75>
    Lifetime=&UniformRangedFloat(Minimum=0.6, Maximum=0.7)
    Module.Write Color=false
    Module.Write Lifetime=true
    Module.Write Mass=false
    Module.Write Material Random=true
    Module.Write Position=false
    Module.Write Scale=false
    Module.Write SpriteRotation=true
    Module.Write SpriteSize=true
    Position=&SimulationPosition()
    Sprite Rotation=&UniformRangedFloat(Minimum=-20, Maximum=20)
    Sprite Size=&MakeCustomVector2DFromBool(Bool=&RandomBool(), True Vector2D=<52 52>, False Vector2D=<-52 52>)
   )
   &SkeletalMeshLocation(Bone Sampling Mode=::NewEnumerator4, Inherit Velocity=false, MaskSpawnGroup=false, Skeletal Mesh=$User.SkeletalMesh)
   &SphereLocation(MaskSpawnGroup=false, Module.Override Randomness=false, Non Uniform Scale=<1 1 0.1>, Sphere Radius=2, Surface Only=false)
   &Color()
   &DynamicMaterialParameters(
    Index 0 Param 2=&UniformRangedFloat(Minimum=0, Maximum=1)
    Index 0 Param 3=1
    Index 0 Param 4=&UniformRangedFloat(Minimum=1, Maximum=2)
    Module.Param0WriteEnabled=false
    Module.Param1WriteEnabled=true
    Module.Param2WriteEnabled=false
    Module.Param3WriteEnabled=true
   ) {module=DynamicMaterialParameters002}
   &FN_SetSpawnGroup(Number of Spawn Groups=2)
  ]
  particleUpdateScript [
   &ParticleState()
   &SkeletalMeshLocation(
    Bone / Socket Index=&SpawnGroup()
    Inherit Velocity=false
    Interpolate Mesh Sampling=false
    MaskSpawnGroup=false
    Mesh Sampling Type=::NewEnumerator5
    Sampled Position Offset=&Add_Vector(
     A=&MaskVectorBySpawnGroup(
      Masked Vector=&RandomRangeVector(Minimum=<0 -5 0>, Maximum=<-10 -5 0>)
      Passthrough Vector=&RandomRangeVector(Minimum=<0 -5 0>, Maximum=<10 -5 0>)
     )
     B=&MaskVectorBySpawnGroup(
      Masked Vector=&VectorFromCurve(
       VectorCurve=DataInterface() {propertyValues {
         XCurve {keys [{arriveTangent=-16.995, interpMode=RCIM_Cubic, leaveTangent=-16.995, tangentMode=RCTM_User, time=0, value=0}, <at 1 -10 RCIM_Cubic>]}
         YCurve {keys [{arriveTangent=-7.1033, interpMode=RCIM_Cubic, leaveTangent=-7.1033, tangentMode=RCTM_User, time=0, value=0}, <at 1 -5 RCIM_Cubic>]}
         ZCurve {keys [{arriveTangent=-28.5033, interpMode=RCIM_Cubic, leaveTangent=-28.5033, tangentMode=RCTM_User, time=0, value=0}, <at 1 -20 RCIM_Cubic>]}
        }}
       Scale Curve=&RandomRangeVector(Minimum=<1 1 0.5>, Maximum=<1 1 1>)
      )
      Passthrough Vector=&VectorFromCurve(
       VectorCurve=DataInterface() {propertyValues {
         XCurve {keys [{arriveTangent=12.7727, interpMode=RCIM_Cubic, leaveTangent=12.7727, tangentMode=RCTM_User, time=0, value=0}, <at 1 10 RCIM_Cubic>]}
         YCurve {keys [{arriveTangent=-6.9601, interpMode=RCIM_Cubic, leaveTangent=-6.9601, tangentMode=RCTM_User, time=0, value=0}, <at 1 -5 RCIM_Cubic>]}
         ZCurve {keys [{arriveTangent=-28.4328, interpMode=RCIM_Cubic, leaveTangent=-28.4328, tangentMode=RCTM_User, time=0, value=0}, <at 1 -20 RCIM_Cubic>]}
        }}
       Scale Curve=&RandomRangeVector(Minimum=<1 1 0.5>, Maximum=<1 1 1>)
      )
     )
    )
    Sampled Position Offset Coordinate Space:ENiagaraMeshTransforms=::NewEnumerator2
    Skeletal Mesh=$User.SkeletalMesh
    Socket Sampling Mode=::NewEnumerator5
   ) {module=SkeletalMeshLocation002}
   &ScaleColor(
    Scale Mode=::NewEnumerator1
    Scale RGBA=&Vector4FromCurve(Vector4Curve=DataInterface() {propertyValues {
       AlphaCurve {keys [<at 0.0024 0>,<at 0.15 1.5>,<at 0.45 1>,<at 0.75 2>,<at 1 0>]}
       BlueCurve {keys [<at 0 0>,<at 0.1017 0>,<at 0.6 0>,<at 0.75 0>]}
       GreenCurve {keys [<at 0 0.25>,<at 0.1017 0.1667>,<at 0.6 0.1333>,<at 0.75 0.0833>]}
       RedCurve {keys [<at 0 1>,<at 0.1017 1>,<at 0.6 1>,<at 0.75 1>]}
      }})
    ScaleRGBA=true
   ) {module=ScaleColor002}
   /SetVariables(Particles.SubImageIndex=&Add_Float(
     A=&RandomRangeFloat(Minimum=0, Maximum=20)
     B=&FloatFromCurve(FloatCurve=DataInterface() {propertyValues {Curve {keys [<at 0 0>,<at 1 24>]}}})
    )) {isSetParameters=true, module=SetVariables_#1}
   &ScaleSpriteSize(Scale Factor=&Vector2DFromFloat(Value=&FloatFromCurve(
      FloatCurve=DataInterface() {propertyValues {Curve {keys [
          {arriveTangent=5.8415, interpMode=RCIM_Cubic, leaveTangent=5.8415, tangentMode=RCTM_User, time=0, value=0}
          {arriveTangent=0.4267, interpMode=RCIM_Cubic, leaveTangent=0.4267, tangentMode=RCTM_User, time=0.25, value=1}
          {arriveTangent=-1.3333, interpMode=RCIM_Cubic, leaveTangent=-1.3333, tangentMode=RCTM_User, time=0.5752, value=1}
          {arriveTangent=-3.118, interpMode=RCIM_Cubic, leaveTangent=-3.118, tangentMode=RCTM_User, time=1, value=0.5}
         ]}}}
      Scale Curve=0.5
     )))
   &DynamicMaterialParameters(
    Index 0 Param 1=&FloatFromCurve(FloatCurve=DataInterface() {propertyValues {Curve {keys [{arriveTangent=2.7551, interpMode=RCIM_Cubic, leaveTangent=2.7551, tangentMode=RCTM_User, time=0, value=0}, <at 0.5 1>]}}})
    Index 0 Param 2=&RandomRangeFloat(Minimum=0.33, Maximum=1)
    Index 0 Param 3=&FloatFromCurve(FloatCurve=DataInterface() {propertyValues {Curve {keys [<at 0 1>,<at 1 2>]}}})
    Index 0 Param 4=1
    Module.Param0WriteEnabled=true
    Module.Param1WriteEnabled=true
    Module.Param2WriteEnabled=true
    Module.Param3WriteEnabled=true
   ) {module=DynamicMaterialParameters001}
   &CameraOffset() {module=CameraOffset001}
   &SpriteRotationRate(Rotation Rate=&RandomRangeFloat(Minimum=-50, Maximum=50))
  ]
  rendererClasses [&NiagaraSpriteRendererProperties]
  renderers [{rendererClass=&NiagaraSpriteRendererProperties}]
 }
 Fire_Plume_World {
  emitterUpdateScript [
   &EmitterState()
   &SpawnRate(Spawn Probability=0.5, SpawnRate=40, Use Spawn Probability=true)
  ]
  particleSpawnScript [
   &InitializeParticle(
    Color=&RandomRangeLinearColor(Minimum=<rgba 0.5 0.5 0.5 0.5>, Maximum=<rgba 10 10 10 1>)
    Lifetime=&RandomRangeFloat(Minimum=1, Maximum=1.25)
    Module.Write Color=true
    Module.Write Lifetime=true
    Module.Write Mass=false
    Module.Write Material Random=true
    Module.Write Position=false
    Module.Write Scale=false
    Module.Write SpriteRotation=true
    Module.Write SpriteSize=true
    Position=&SimulationPosition()
    Sprite Rotation=&NormalizedAngleToDegrees(NormalizedAngle=&RandomRangeFloat(Minimum=-1, Maximum=1))
    Sprite Size=&Vector2DFromFloat(Value=&RandomRangeFloat(Minimum=12, Maximum=26))
   )
   &SkeletalMeshLocation(
    Bone Sampling Mode=::NewEnumerator4
    Inherit Velocity=false
    Interpolate Mesh Sampling=false
    MaskSpawnGroup=false
    Sampled Position Offset Coordinate Space:ENiagaraMeshTransforms=::NewEnumerator3
    Skeletal Mesh=$User.SkeletalMesh
   )
   &BoxLocation(Box Offset=<0 -4 3>, Box Size=<30 10 3>, MaskSpawnGroup=false, Module.Override Seed=false, Rotate Box=false, Surface Only=false)
   &CameraOffset(Camera Offset Amount=-3)
   &FN_SetSpawnGroup(Number of Spawn Groups=2)
   &InheritVelocity(Inherited Velocity Amount Scale=&VectorFromFloat(Value=-1), Speed Limit=true, UseSpeedThreshold=false) {enabled=false}
  ]
  particleUpdateScript [
   &ParticleState()
   &SkeletalMeshLocation(
    Bone / Socket Index=&SpawnGroup()
    Inherit Velocity=false
    Interpolate Mesh Sampling=false
    MaskSpawnGroup=false
    Mesh Sampling Type=::NewEnumerator5
    Sampled Position Offset=&Add_Vector(
     A=&MaskVectorBySpawnGroup(
      Masked Vector=&RandomRangeVector(Minimum=<0 -3 0>, Maximum=<-15 -2 -2>)
      Passthrough Vector=&RandomRangeVector(Minimum=<0 -3 0>, Maximum=<15 -2 -2>)
     )
     B=&MaskVectorBySpawnGroup(
      Masked Vector=&VectorFromCurve(
       VectorCurve=DataInterface() {propertyValues {
         XCurve {keys [{arriveTangent=-16.995, interpMode=RCIM_Cubic, leaveTangent=-16.995, tangentMode=RCTM_User, time=0, value=0}, <at 1 -7 RCIM_Cubic>]}
         YCurve {keys [{arriveTangent=-7.5962, interpMode=RCIM_Cubic, leaveTangent=-7.5962, tangentMode=RCTM_User, time=0, value=0}, <at 0.8 -7 RCIM_Cubic>]}
         ZCurve {keys [{arriveTangent=18.6381, interpMode=RCIM_Cubic, leaveTangent=18.6381, tangentMode=RCTM_User, time=0, value=0}, <at 1 11 RCIM_Cubic>]}
        }}
       Scale Curve=&RandomRangeVector(Minimum=<1 1 -1>, Maximum=<1 1 1>)
      )
      Passthrough Vector=&VectorFromCurve(
       VectorCurve=DataInterface() {propertyValues {
         XCurve {keys [{arriveTangent=12.7727, interpMode=RCIM_Cubic, leaveTangent=12.7727, tangentMode=RCTM_User, time=0, value=0}, <at 1 7 RCIM_Cubic>]}
         YCurve {keys [{arriveTangent=-5.3295, interpMode=RCIM_Cubic, leaveTangent=-5.3295, tangentMode=RCTM_User, time=0, value=0}, <at 1 -7 RCIM_Cubic>]}
         ZCurve {keys [{arriveTangent=26.6251, interpMode=RCIM_Cubic, leaveTangent=26.6251, tangentMode=RCTM_User, time=0, value=0}, <at 0.8 11 RCIM_Cubic>]}
        }}
       Scale Curve=&RandomRangeVector(Minimum=<1 1 -1>, Maximum=<1 1 1>)
      )
     )
    )
    Sampled Position Offset Coordinate Space:ENiagaraMeshTransforms=::NewEnumerator2
    Skeletal Mesh=$User.SkeletalMesh
    Socket Sampling Mode=::NewEnumerator5
   ) {module=SkeletalMeshLocation002}
   &ScaleColor(
    Scale Mode=::NewEnumerator1
    Scale RGBA=&Vector4FromCurve(Vector4Curve=DataInterface() {propertyValues {
       AlphaCurve {keys [
         <at 0 0 RCIM_Cubic>
         {arriveTangent=3.3333, interpMode=RCIM_Cubic, leaveTangent=3.3333, time=0.3, value=1}
         {arriveTangent=-1.6667, interpMode=RCIM_Cubic, leaveTangent=-1.6667, time=0.6, value=2}
         {arriveTangent=-0.4881, interpMode=RCIM_Cubic, leaveTangent=-0.4881, tangentMode=RCTM_User, time=0.9, value=0}
        ]}
       BlueCurve {keys [<at 0 0>,<at 0.276 0>,<at 0.8749 0>,<at 1 0>]}
       GreenCurve {keys [<at 0 0>,<at 0.276 0.0235>,<at 0.8749 0.2319>,<at 1 1.4209>]}
       RedCurve {keys [<at 0 0.1>,<at 0.276 1>,<at 0.8749 1>,<at 1 2>]}
      }})
    ScaleRGBA=true
   )
   &ScaleColor(Scale Alpha=0.4, ScaleA=true, ScaleRGB=false) {module=ScaleColor001}
   &DynamicMaterialParameters(
    Index 0 Param 1=&FloatFromCurve(FloatCurve=DataInterface() {propertyValues {Curve {keys [{arriveTangent=2.7551, interpMode=RCIM_Cubic, leaveTangent=2.7551, tangentMode=RCTM_User, time=0, value=0}, <at 0.6 1>]}}})
    Index 0 Param 2=&RandomRangeFloat(Minimum=0.33, Maximum=1)
    Index 0 Param 3=&FloatFromCurve(FloatCurve=DataInterface() {propertyValues {Curve {keys [
         {arriveTangent=6.3026, interpMode=RCIM_Cubic, time=0, value=1}
         {arriveTangent=2.4291, interpMode=RCIM_Cubic, leaveTangent=6.3584, tangentMode=RCTM_User, time=1, value=2}
        ]}}})
    Index 0 Param 4=&RandomRangeFloat(Minimum=1, Maximum=2)
    Module.Param0WriteEnabled=true
    Module.Param1WriteEnabled=true
    Module.Param2WriteEnabled=true
    Module.Param3WriteEnabled=true
   ) {module=DynamicMaterialParameters001}
   &ScaleSpriteSize(Scale Factor=&Vector2DFromFloat(Value=&FloatFromCurve(FloatCurve=DataInterface() {propertyValues {Curve {keys [
          {arriveTangent=-0.0804, interpMode=RCIM_Cubic, leaveTangent=-0.0804, tangentMode=RCTM_User, time=0, value=0.75}
          {interpMode=RCIM_Cubic, tangentMode=RCTM_User, time=0.4, value=1}
          {arriveTangent=-4.5991, interpMode=RCIM_Cubic, leaveTangent=-12.9762, tangentMode=RCTM_User, time=1, value=0}
         ]}}})))
   &SpriteRotationRate(Rotation Rate=20)
   &SubUVAnimation()
   &AccelerationForce(Acceleration=&SelectVectorFromArray(
     Array Sampling Mode=::NewEnumerator0
     Direct Array Index=?
     Vector Selection Array=DataInterface() {propertyValues {FloatData [<0 0 70>, <0 0 -70>]}}
     Random Seed=0
    ))
   &Drag(Drag=&RandomRangeFloat(Minimum=4, Maximum=4), Use Linear Drag=true, Use Rotational Drag=false)
   &SolveForcesAndVelocity(Clamp Velocity=false, Limit Acceleration=false)
  ]
  rendererClasses [&NiagaraSpriteRendererProperties]
  renderers [{rendererClass=&NiagaraSpriteRendererProperties}]
 }
 Light {
  emitterUpdateScript [
   &EmitterState(Inactive Response=::NewEnumerator1, Life Cycle Mode=::NewEnumerator1, Loop Duration=2, UseLoopDelay=true)
   &SpawnBurst_Instantaneous(Spawn Count=1, Use Spawn Probability=false)
  ]
  particleSpawnScript [
   &InitializeParticle(
    Color=<rgba 4 2.0181 0.0361 1>
    Lifetime=2
    Module.Write Color=true
    Module.Write Lifetime=true
    Module.Write Mass=false
    Module.Write Material Random=true
    Module.Write Position=false
    Module.Write Scale=false
    Module.Write SpriteRotation=false
    Module.Write SpriteSize=true
    Position=&SimulationPosition()
    Sprite Size=<75 75>
   )
   &DynamicMaterialParameters(
    Index 0 Param 1=&UniformRangedFloat(Minimum=0.7, Maximum=0.9)
    Index 0 Param 2=&UniformRangedFloat(Minimum=0.05, Maximum=0.15)
    Index 0 Param 3=1
    Module.Param0WriteEnabled=true
    Module.Param1WriteEnabled=true
    Module.Param2WriteEnabled=false
    Module.Param3WriteEnabled=false
   ) {module=DynamicMaterialParameters002}
  ]
  particleUpdateScript [
   &ParticleState()
   &SkeletalMeshLocation(
    Bone Sampling Mode=::NewEnumerator4
    Inherit Velocity=false
    MaskSpawnGroup=false
    Sampled Position Offset=<7 0 16>
    Sampled Position Offset Coordinate Space:ENiagaraMeshTransforms=::NewEnumerator2
    Skeletal Mesh=$User.SkeletalMesh
   )
   &SolveForcesAndVelocity(Clamp Velocity=false, Limit Acceleration=false)
   &ScaleColor(
    Scale RGB=&Multiply_Vector(
     A=&MakeVectorFromLinearColorRGB(Color=&LerpLinearColors(
       StartColor=<rgba 2 1.5765 0.3059 1>
       EndColor=<rgba 2 0.8 0 1>
       LerpFactor=&FloatFromCurve(
        FloatCurve=DataInterface() {propertyValues {Curve {keys [<at 0 1>,<at 0.1069 0.0895>,<at 0.3206 0.7423>,<at 0.4907 0.1444>,<at 0.6992 0.8948>,<at 0.9061 0.1261>,<at 1 1>]}}}
        CurveIndex=$Emitter.NormalizedLoopAge
        Scale Curve=0.9
       )
      ))
     B=&VectorFromFloat(Value=&FloatFromCurve(
       FloatCurve=DataInterface() {propertyValues {Curve {keys [<at 0 1>,<at 0.0584 1.1216>,<at 0.0797 0.6691>,<at 0.1464 0.526>,<at 0.222 1.0327>,<at 0.2705 0.6844>,<at 0.3166 0.8798>,<at 0.3641 0.5477>,<at 0.4407 0.6405>,<at 0.4815 1.1358>,<at 0.5388 1.0006>,<at 0.58 0.5314>,<at 0.6479 0.8407>,<at 0.6937 0.7459>,<at 0.737 1.0896>,<at 0.8125 0.5703>,<at 0.8909 0.8732>,<at 0.9449 0.6186>,<at 1 1>]}}}
       CurveIndex=$Emitter.NormalizedLoopAge
       Scale Curve=0.2
      ))
    )
    ScaleA=true
    ScaleRGB=true
   )
   &FN_Light_v2(Light Exponent=2, Light Radius=40, LightEnabled=true, Scale Color=1)
   &ScaleSpriteSize(Scale Factor=&Vector2DFromFloat(Value=&FloatFromCurve(
      FloatCurve=DataInterface() {propertyValues {Curve {keys [<at 0 1>,<at 0.0562 0.9129>,<at 0.0748 1.1321>,<at 0.1235 1.3668>,<at 0.2098 0.8951>,<at 0.2589 1.1544>,<at 0.2987 1.0516>,<at 0.3338 1.2874>,<at 0.3749 1.3981>,<at 0.4815 0.8928>,<at 0.5082 1.113>,<at 0.5446 1.2763>,<at 0.6237 1.1354>,<at 0.6838 1.0572>,<at 0.7109 0.8761>,<at 0.7888 0.7005>,<at 0.8531 1.2103>,<at 0.9284 1.1544>,<at 1 1>]}}}
      CurveIndex=$Emitter.NormalizedLoopAge
     )))
  ]
  rendererClasses [&NiagaraLightRendererProperties]
  renderers [{rendererClass=&NiagaraLightRendererProperties, rendererIndex=0}]
 }
 Tail {
  emitterUpdateScript [
   &EmitterState(Life Cycle Mode=::NewEnumerator1, Loop Behavior=::NewEnumerator1, Loop Duration=0.1, UseLoopDelay=false)
   &SpawnBurst_Instantaneous(Spawn Count=1, Use Spawn Probability=false)
   &SpawnBurst_Instantaneous(Spawn Count=3, Spawn Time=0.2374, Use Spawn Probability=false, UseLoopCountLimit=false) {enabled=false, module=SpawnBurst_Instantaneous001}
  ]
  particleSpawnScript [
   &InitializeParticle(
    Color=<rgba 1.7375 0 2 0.2>
    Mesh Scale=&Multiply_VectorByFloat(
     Vector=<1 0.8 1>
     Float=&SelectFloatFromArray(
      Direct Array Index=&ReturnExecIndex()
      Float Selection Array=DataInterface() {propertyValues {FloatData [0.8,0.8,0.3]}}
     )
    )
    Module.Write Color=false
    Module.Write Lifetime=true
    Module.Write Mass=false
    Module.Write Material Random=true
    Module.Write Position=false
    Module.Write Scale=true
    Module.Write SpriteRotation=false
    Module.Write SpriteSize=true
    Sprite Size=<8 8>
   ) {enabled=false}
   /Niagara/Modules/Spawn/Initialization/V2/InitializeParticle(
    Mesh Scale=&Multiply_VectorByFloat(Vector=<1 0.4 2>, Float=0.3)
    Mesh Scale Mode=::NewEnumerator4
    Position Offset=&SelectVectorFromArray(
     Direct Array Index=&ReturnExecIndex()
     Vector Selection Array=DataInterface() {propertyValues {FloatData [<0 -3 0>, <0 0 0>, <0 0 0>]}}
    )
    Position Offset Coordinate Space=::Local
    UsePositionOffset=true
    Write Material Random=true
   ) {module=InitializeParticle001}
   &InitialMeshOrientation(Mesh Orientation Mode=::NewEnumerator1, Rotation=<0.85 -0.95 0.5>, Rotation Coordinate Space=::NewEnumerator0, Use Rotation Vector=true)
  ]
  particleUpdateScript [
   &ParticleState(Kill Particles When Lifetime Has Elapsed=false, Loop Particles Lifetime=true)
   &SkeletalMeshLocation(Bone Sampling Mode=::NewEnumerator4, MaskSpawnGroup=false, Sampled Position Offset=<0.5 -3 -1>, Skeletal Mesh=$User.SkeletalMesh) {enabled=false, module=SkeletalMeshLocation001}
   /SetVariables(Particles.LightCol:LinearColor=&MultiplyLinearColor(
     Linear Color=<rgba 0.5 0.089 0 1>
     Scale Factor=&MakeVector4_FromFloat(Float=&Sine(Normalized Angle=$Emitter.Age, Period=0.2, Scale=0.1, Bias=2))
    )) {isSetParameters=true, module=SetVariables_#2}
   &ScaleColor(
    Scale Mode=::NewEnumerator1
    Scale RGBA=&MakeVector4_FromLinearColor(LinearColor=&ColorFromCurve(ColorCurve=DataInterface() {propertyValues {
        AlphaCurve {keys [<at 0 0.5>,<at 0.5 1>,<at 1 0.5>]}
        BlueCurve {keys [<at 0 1 RCIM_Cubic>,<at 0.5 2 RCIM_Cubic>,<at 1 1 RCIM_Cubic>]}
        GreenCurve {keys [<at 0 1 RCIM_Cubic>,<at 0.5 2 RCIM_Cubic>,<at 1 1 RCIM_Cubic>]}
        RedCurve {keys [<at 0 1 RCIM_Cubic>,<at 0.5 2 RCIM_Cubic>,<at 1 1 RCIM_Cubic>]}
       }}))
    ScaleRGBA=true
   ) {enabled=false}
   &DynamicMaterialParameters(
    Index 0 Param 3=10
    Module.Param0WriteEnabled=true
    Module.Param0WriteEnabled1=true
    Module.Param1WriteEnabled=true
    Module.Param1WriteEnabled1=true
    Module.Param2WriteEnabled=true
    Module.Param2WriteEnabled1=true
    Module.Param3WriteEnabled=true
    Module.Param3WriteEnabled1=true
    Write Parameter Index 1=true
   )
  ]
  rendererClasses [
   &NiagaraSpriteRendererProperties
   &NiagaraMeshRendererProperties
   &NiagaraLightRendererProperties
  ]
  renderers [
   {rendererClass=&NiagaraSpriteRendererProperties}
   {rendererClass=&NiagaraMeshRendererProperties, rendererIndex=1}
   {rendererClass=&NiagaraLightRendererProperties, rendererIndex=2}
  ]
 }
 Wings {
  emitterUpdateScript [
   &EmitterState(Life Cycle Mode=::NewEnumerator1, Loop Behavior=::NewEnumerator1, Loop Duration=0.1, UseLoopDelay=false)
   &SpawnBurst_Instantaneous(Spawn Count=2, Use Spawn Probability=false)
   &SpawnBurst_Instantaneous(Spawn Count=3, Spawn Time=0.2374, Use Spawn Probability=false, UseLoopCountLimit=false) {enabled=false, module=SpawnBurst_Instantaneous001}
  ]
  particleSpawnScript [
   /Niagara/Modules/Spawn/Initialization/InitializeParticle(
    Color=<rgba 1.7375 0 2 0.2>
    Mesh Scale=&Multiply_VectorByFloat(
     Vector=<1 0.8 1>
     Float=&SelectFloatFromArray(
      Direct Array Index=&ReturnExecIndex()
      Float Selection Array=DataInterface() {propertyValues {FloatData [0.8,0.8,0.3]}}
     )
    )
    Module.Write Color=false
    Module.Write Lifetime=true
    Module.Write Mass=false
    Module.Write Material Random=true
    Module.Write Position=false
    Module.Write Scale=true
    Module.Write SpriteRotation=false
    Module.Write SpriteSize=true
    Sprite Size=<8 8>
   ) {enabled=false}
   /Niagara/Modules/Spawn/Initialization/V2/InitializeParticle(
    Mesh Scale=&Multiply_VectorByFloat(
     Vector=<0.7 0.5 1>
     Float=&SelectFloatFromArray(
      Direct Array Index=&ReturnExecIndex()
      Float Selection Array=DataInterface() {propertyValues {FloatData [0.5,0.5,0.3]}}
     )
    )
    Mesh Scale Mode=::NewEnumerator4
    Position Offset=&SelectVectorFromArray(
     Direct Array Index=&ReturnExecIndex()
     Vector Selection Array=DataInterface() {propertyValues {FloatData [<0 -2.5 0>, <0 -2.5 0>, <0 0 0>]}}
    )
    Position Offset Coordinate Space=::Local
    UsePositionOffset=true
    Write Material Random=true
   ) {module=InitializeParticle001}
   &InitialMeshOrientation(
    Mesh Orientation Mode=::NewEnumerator1
    Rotation=&SelectVectorFromArray(
     Direct Array Index=&ReturnExecIndex()
     Vector Selection Array=DataInterface() {propertyValues {FloatData [<0.25 0 0>, <-0.01 -0.37 0.35>, <3.3239 2.6261 0.1263>]}}
    )
    Rotation Coordinate Space=::NewEnumerator2
    Use Rotation Vector=true
   )
  ]
  particleUpdateScript [
   &ParticleState(Kill Particles When Lifetime Has Elapsed=false, Loop Particles Lifetime=true)
   &SkeletalMeshLocation(
    Bone Sampling Mode=::NewEnumerator4
    MaskSpawnGroup=false
    Sampled Position Offset=&SelectVectorFromArray(
     Direct Array Index=&ReturnExecIndex()
     Vector Selection Array=DataInterface() {propertyValues {FloatData [<0 -2 0>, <0 -2 0>, <0 0 -5>]}}
    )
    Skeletal Mesh=$User.SkeletalMesh
   ) {enabled=false, module=SkeletalMeshLocation001}
   /SetVariables(Particles.LightCol:LinearColor=&MultiplyLinearColor(
     Linear Color=<rgba 0.5 0.089 0 1>
     Scale Factor=&MakeVector4_FromFloat(Float=&Sine(Normalized Angle=$Emitter.Age, Period=0.2, Scale=0.1, Bias=2))
    )) {isSetParameters=true, module=SetVariables_#2}
   &ScaleColor(
    Scale Mode=::NewEnumerator1
    Scale RGBA=&MakeVector4_FromLinearColor(LinearColor=&ColorFromCurve(ColorCurve=DataInterface() {propertyValues {
        AlphaCurve {keys [<at 0 0.5>,<at 0.5 1>,<at 1 0.5>]}
        BlueCurve {keys [<at 0 1 RCIM_Cubic>,<at 0.5 2 RCIM_Cubic>,<at 1 1 RCIM_Cubic>]}
        GreenCurve {keys [<at 0 1 RCIM_Cubic>,<at 0.5 2 RCIM_Cubic>,<at 1 1 RCIM_Cubic>]}
        RedCurve {keys [<at 0 1 RCIM_Cubic>,<at 0.5 2 RCIM_Cubic>,<at 1 1 RCIM_Cubic>]}
       }}))
    ScaleRGBA=true
   ) {enabled=false}
   &DynamicMaterialParameters(
    Index 0 Param 3=10
    Module.Param0WriteEnabled=true
    Module.Param0WriteEnabled1=true
    Module.Param1WriteEnabled=true
    Module.Param1WriteEnabled1=true
    Module.Param2WriteEnabled=true
    Module.Param2WriteEnabled1=true
    Module.Param3WriteEnabled=true
    Module.Param3WriteEnabled1=true
    Write Parameter Index 1=true
   )
  ]
  rendererClasses [
   &NiagaraSpriteRendererProperties
   &NiagaraMeshRendererProperties
   &NiagaraLightRendererProperties
  ]
  renderers [
   {rendererClass=&NiagaraSpriteRendererProperties}
   {rendererClass=&NiagaraMeshRendererProperties, rendererIndex=1}
   {rendererClass=&NiagaraLightRendererProperties, rendererIndex=2}
  ]
 }
 Yellow_FrontFlames {
  bEnabled=false
  emitterUpdateScript [
   &EmitterState()
   &SpawnRate(SpawnRate=16, Use Spawn Probability=false)
  ]
  particleSpawnScript [
   /Niagara/Modules/Spawn/Initialization/InitializeParticle(
    Color=<rgba 5 3.0056 1.6206 0.75>
    Lifetime=&UniformRangedFloat(Minimum=0.55, Maximum=0.66)
    Module.Write Color=false
    Module.Write Lifetime=true
    Module.Write Mass=false
    Module.Write Material Random=true
    Module.Write Position=false
    Module.Write Scale=false
    Module.Write SpriteRotation=true
    Module.Write SpriteSize=true
    Position=&SimulationPosition()
    Sprite Rotation=&UniformRangedFloat(Minimum=0, Maximum=360)
    Sprite Size=&MakeCustomVector2DFromBool(Bool=&RandomBool(), True Vector2D=<28 28>, False Vector2D=<-28 28>)
   )
   &SkeletalMeshLocation(
    Apply To Particle Position=false
    Bone Sampling Mode=::NewEnumerator4
    Inherit Velocity=false
    MaskSpawnGroup=false
    Skeletal Mesh=$User.SkeletalMesh
   )
   &SphereLocation(MaskSpawnGroup=false, Module.Override Randomness=false, Non Uniform Scale=<1 1 0.1>, Sphere Radius=20, Surface Only=false)
   &Color()
   &DynamicMaterialParameters(
    Index 0 Param 2=&UniformRangedFloat(Minimum=0, Maximum=1)
    Index 0 Param 3=1
    Index 0 Param 4=&UniformRangedFloat(Minimum=1, Maximum=2)
    Module.Param0WriteEnabled=false
    Module.Param1WriteEnabled=true
    Module.Param2WriteEnabled=false
    Module.Param3WriteEnabled=true
   ) {module=DynamicMaterialParameters002}
  ]
  particleUpdateScript [
   &ParticleState()
   &SkeletalMeshLocation(
    Apply To Particle Position=false
    Bone Sampling Mode=::NewEnumerator4
    Inherit Velocity=false
    MaskSpawnGroup=false
    Sampled Position Offset:NiagaraPosition=$Particles.Initial.Position
    Skeletal Mesh=$User.SkeletalMesh
   ) {module=SkeletalMeshLocation001}
   &self:LERP-Source_to_Destination(
    DestinationPosition:NiagaraPosition=$Particles.Initial.Position
    DestinationPosition-Offset:Vector3f=&RandomRangeVector(
     Minimum=&VectorFromCurve(
      VectorCurve=DataInterface() {propertyValues {
        XCurve {keys [{arriveTangent=19.163, interpMode=RCIM_Cubic, leaveTangent=19.163, tangentMode=RCTM_User, time=0, value=0}]}
        YCurve {keys [<at 0 -3>]}
        ZCurve {keys [<at 0 0 RCIM_Cubic>, {arriveTangent=42.1643, interpMode=RCIM_Cubic, leaveTangent=42.1645, tangentMode=RCTM_User, time=1, value=30}]}
       }}
      Scale Curve=<1 1 0.7>
     )
     Maximum=&VectorFromCurve(
      VectorCurve=DataInterface() {propertyValues {
        XCurve {keys [{arriveTangent=-4.6571, interpMode=RCIM_Cubic, leaveTangent=-4.6571, tangentMode=RCTM_User, time=0, value=0}]}
        YCurve {keys [<at 0 3>]}
        ZCurve {keys [<at 0 0 RCIM_Cubic>, {arriveTangent=47.8845, interpMode=RCIM_Cubic, leaveTangent=47.885, tangentMode=RCTM_User, time=1, value=30}]}
       }}
      Scale Curve=<1 1 0.55>
     )
    )
    Normalized Influence:NiagaraFloat=&MakeCustomFloatFromBool(
     Bool=$User.IsSkydiving
     True Float=&FloatFromCurve(FloatCurve=DataInterface() {propertyValues {Curve {keys [
          {arriveTangent=0.0075, interpMode=RCIM_Cubic, leaveTangent=0.0075, tangentMode=RCTM_User, time=0, value=0}
          {arriveTangent=0.0813, interpMode=RCIM_Cubic, leaveTangent=0.0813, tangentMode=RCTM_User, time=1, value=0.025}
         ]}}})
     False Float=&FloatFromCurve(FloatCurve=DataInterface() {propertyValues {Curve {keys [
          {arriveTangent=0.037, interpMode=RCIM_Cubic, leaveTangent=0.037, tangentMode=RCTM_User, time=0, value=0}
          {arriveTangent=0.2447, interpMode=RCIM_Cubic, leaveTangent=0.2447, tangentMode=RCTM_User, time=1, value=0.1}
         ]}}})
    )
    SourcePosition:Vector3f=$Output.SkeletalMeshLocation001.SampledPosition
    SourcePosition-Offset:Vector3f=&TransformVector(Vector=&RandomRangeVector(Minimum=<15 -1.5 0>, Maximum=<-15 -1.5 -5>), Destination Space=::Simulation)
   ) {module=LERP-Source_to_Destination}
   &ScaleColor(
    Scale Mode=::NewEnumerator1
    Scale RGBA=&Vector4FromCurve(Vector4Curve=DataInterface() {propertyValues {
       AlphaCurve {keys [<at 0.0024 0>,<at 0.1271 0.5>,<at 0.75 5>,<at 1 0>]}
       BlueCurve {keys [<at 0.2365 0>,<at 0.65 0.016>,<at 0.85 0>]}
       GreenCurve {keys [<at 0.2365 0.2858>,<at 0.65 0.1964>,<at 0.85 0.0833>]}
       RedCurve {keys [<at 0.2365 1>,<at 0.65 1>,<at 0.85 1>]}
      }})
    ScaleRGBA=true
   )
   /SetVariables(Particles.SubImageIndex=&FloatFromCurve(FloatCurve=DataInterface() {propertyValues {Curve {keys [<at 0 0>,<at 1 34>]}}})) {isSetParameters=true, module=SetVariables_#1}
   &ScaleSpriteSize(Scale Factor=&Vector2DFromFloat(Value=&FloatFromCurve(
      FloatCurve=DataInterface() {propertyValues {Curve {keys [
          {arriveTangent=7.5764, interpMode=RCIM_Cubic, leaveTangent=7.5765, tangentMode=RCTM_User, time=0, value=0.02}
          {arriveTangent=-0.0353, interpMode=RCIM_Cubic, leaveTangent=-0.0353, tangentMode=RCTM_User, time=0.2, value=1.2}
          <at 0.5752 1.1643>
          <at 1 0.65 RCIM_Cubic>
         ]}}}
      Scale Curve=0.5
     )))
   &DynamicMaterialParameters(
    Index 0 Param 1=&FloatFromCurve(FloatCurve=DataInterface() {propertyValues {Curve {keys [{arriveTangent=2.7551, interpMode=RCIM_Cubic, leaveTangent=2.7551, tangentMode=RCTM_User, time=0, value=0}, <at 0.5 1>]}}})
    Index 0 Param 2=&RandomRangeFloat(Minimum=0.33, Maximum=1)
    Index 0 Param 3=&FloatFromCurve(FloatCurve=DataInterface() {propertyValues {Curve {keys [<at 0.8 1>,<at 1 2>]}}})
    Index 0 Param 4=1
    Module.Param0WriteEnabled=true
    Module.Param1WriteEnabled=true
    Module.Param2WriteEnabled=true
    Module.Param3WriteEnabled=true
   ) {module=DynamicMaterialParameters001}
   &CameraOffset(Camera Offset Amount=&FloatFromCurve(FloatCurve=DataInterface() {propertyValues {Curve {keys [<at 0.35 0>,<at 0.7 6>]}}})) {module=CameraOffset001}
  ]
  rendererClasses [&NiagaraSpriteRendererProperties]
  renderers [{rendererClass=&NiagaraSpriteRendererProperties}]
 }
 Yellow_FrontFlames001 {
  emitterUpdateScript [
   &EmitterState()
   &SpawnRate(SpawnRate=16, Use Spawn Probability=false)
  ]
  particleSpawnScript [
   &InitializeParticle(
    Color=<rgba 5 3.0056 1.6206 0.75>
    Lifetime=&UniformRangedFloat(Minimum=0.55, Maximum=0.66)
    Module.Write Color=false
    Module.Write Lifetime=true
    Module.Write Mass=false
    Module.Write Material Random=true
    Module.Write Position=false
    Module.Write Scale=false
    Module.Write SpriteRotation=true
    Module.Write SpriteSize=true
    Position=&SimulationPosition()
    Sprite Rotation=&UniformRangedFloat(Minimum=0, Maximum=360)
    Sprite Size=&MakeCustomVector2DFromBool(Bool=&RandomBool(), True Vector2D=<28 28>, False Vector2D=<-28 28>)
   )
   &SkeletalMeshLocation(
    Apply To Particle Position=false
    Bone Sampling Mode=::NewEnumerator4
    Inherit Velocity=false
    MaskSpawnGroup=false
    Skeletal Mesh=$User.SkeletalMesh
   )
   &SphereLocation(MaskSpawnGroup=false, Module.Override Randomness=false, Non Uniform Scale=<1 1 0.1>, Sphere Radius=20, Surface Only=false)
   &Color()
   &DynamicMaterialParameters(
    Index 0 Param 2=&UniformRangedFloat(Minimum=0, Maximum=1)
    Index 0 Param 3=1
    Index 0 Param 4=&UniformRangedFloat(Minimum=1, Maximum=2)
    Module.Param0WriteEnabled=false
    Module.Param1WriteEnabled=true
    Module.Param2WriteEnabled=false
    Module.Param3WriteEnabled=true
   ) {module=DynamicMaterialParameters002}
   &FN_SetSpawnGroup(Number of Spawn Groups=2)
  ]
  particleUpdateScript [
   &ParticleState()
   &SkeletalMeshLocation(
    Bone / Socket Index=&SpawnGroup()
    Inherit Velocity=false
    Interpolate Mesh Sampling=false
    MaskSpawnGroup=false
    Mesh Sampling Type=::NewEnumerator5
    Sampled Position Offset=&Add_Vector(
     A=&MaskVectorBySpawnGroup(
      Masked Vector=&RandomRangeVector(Minimum=<0 -3 0>, Maximum=<-15 -3 5>)
      Passthrough Vector=&RandomRangeVector(Minimum=<0 -3 0>, Maximum=<15 -3 5>)
     )
     B=&MaskVectorBySpawnGroup(
      Masked Vector=&VectorFromCurve(
       VectorCurve=DataInterface() {propertyValues {
         XCurve {keys [{arriveTangent=-16.995, interpMode=RCIM_Cubic, leaveTangent=-16.995, tangentMode=RCTM_User, time=0, value=0}, <at 1 -10 RCIM_Cubic>]}
         YCurve {keys [{arriveTangent=-7.5962, interpMode=RCIM_Cubic, leaveTangent=-7.5962, tangentMode=RCTM_User, time=0, value=0}, <at 1 -7 RCIM_Cubic>]}
         ZCurve {keys [{arriveTangent=18.6381, interpMode=RCIM_Cubic, leaveTangent=18.6381, tangentMode=RCTM_User, time=0, value=0}, <at 1 8 RCIM_Cubic>]}
        }}
       Scale Curve=&RandomRangeVector(Minimum=<1 1 0.5>, Maximum=<1 1 1>)
      )
      Passthrough Vector=&VectorFromCurve(
       VectorCurve=DataInterface() {propertyValues {
         XCurve {keys [{arriveTangent=12.7727, interpMode=RCIM_Cubic, leaveTangent=12.7727, tangentMode=RCTM_User, time=0, value=0}, <at 1 10 RCIM_Cubic>]}
         YCurve {keys [{arriveTangent=-5.3295, interpMode=RCIM_Cubic, leaveTangent=-5.3295, tangentMode=RCTM_User, time=0, value=0}, <at 1 -7 RCIM_Cubic>]}
         ZCurve {keys [{arriveTangent=26.6251, interpMode=RCIM_Cubic, leaveTangent=26.6251, tangentMode=RCTM_User, time=0, value=0}, <at 1 8 RCIM_Cubic>]}
        }}
       Scale Curve=&RandomRangeVector(Minimum=<1 1 0.5>, Maximum=<1 1 1>)
      )
     )
    )
    Sampled Position Offset Coordinate Space:ENiagaraMeshTransforms=::NewEnumerator2
    Skeletal Mesh=$User.SkeletalMesh
    Socket Sampling Mode=::NewEnumerator5
   ) {module=SkeletalMeshLocation002}
   &ScaleColor(
    Scale Mode=::NewEnumerator1
    Scale RGBA=&Vector4FromCurve(Vector4Curve=DataInterface() {propertyValues {
       AlphaCurve {keys [<at 0.0024 0>,<at 0.1271 0.5>,<at 0.75 5>,<at 1 0>]}
       BlueCurve {keys [<at 0.2365 0>,<at 0.65 0.016>,<at 0.85 0>]}
       GreenCurve {keys [<at 0.2365 0.2858>,<at 0.65 0.1964>,<at 0.85 0.0833>]}
       RedCurve {keys [<at 0.2365 1>,<at 0.65 1>,<at 0.85 1>]}
      }})
    ScaleRGBA=true
   )
   /SetVariables(Particles.SubImageIndex=&FloatFromCurve(FloatCurve=DataInterface() {propertyValues {Curve {keys [<at 0 0>,<at 1 34>]}}})) {isSetParameters=true, module=SetVariables_#1}
   &ScaleSpriteSize(Scale Factor=&Vector2DFromFloat(Value=&FloatFromCurve(
      FloatCurve=DataInterface() {propertyValues {Curve {keys [
          {arriveTangent=7.5764, interpMode=RCIM_Cubic, leaveTangent=7.5765, tangentMode=RCTM_User, time=0, value=0.02}
          {arriveTangent=-0.0353, interpMode=RCIM_Cubic, leaveTangent=-0.0353, tangentMode=RCTM_User, time=0.2, value=1.2}
          <at 0.5752 1.1643>
          <at 1 0.65 RCIM_Cubic>
         ]}}}
      Scale Curve=0.5
     )))
   &DynamicMaterialParameters(
    Index 0 Param 1=&FloatFromCurve(FloatCurve=DataInterface() {propertyValues {Curve {keys [{arriveTangent=2.7551, interpMode=RCIM_Cubic, leaveTangent=2.7551, tangentMode=RCTM_User, time=0, value=0}, <at 0.5 1>]}}})
    Index 0 Param 2=&RandomRangeFloat(Minimum=0.33, Maximum=1)
    Index 0 Param 3=&FloatFromCurve(FloatCurve=DataInterface() {propertyValues {Curve {keys [<at 0.8 1>,<at 1 2>]}}})
    Index 0 Param 4=1
    Module.Param0WriteEnabled=true
    Module.Param1WriteEnabled=true
    Module.Param2WriteEnabled=true
    Module.Param3WriteEnabled=true
   ) {module=DynamicMaterialParameters001}
   &CameraOffset(Camera Offset Amount=&FloatFromCurve(FloatCurve=DataInterface() {propertyValues {Curve {keys [<at 0.35 0>,<at 0.7 6>]}}})) {module=CameraOffset001}
  ]
  rendererClasses [&NiagaraSpriteRendererProperties]
  renderers [{rendererClass=&NiagaraSpriteRendererProperties}]
 }
 decal001 {
  emitterSpawnScript [/SetVariables(Emitter.DepthFade=25) {isSetParameters=true}]
  emitterUpdateScript [
   &EmitterState()
   &SpawnRate(SpawnRate=4, Use Spawn Probability=false)
  ]
  particleSpawnScript [
   &InitializeParticle(
    Color=<rgba 5 3.0056 1.6206 0.75>
    Lifetime=&UniformRangedFloat(Minimum=0.5, Maximum=1)
    Module.Write Color=false
    Module.Write Lifetime=true
    Module.Write Mass=false
    Module.Write Material Random=true
    Module.Write Position=false
    Module.Write Scale=false
    Module.Write SpriteRotation=false
    Module.Write SpriteSize=true
    Position=&SimulationPosition()
    Sprite Rotation=&UniformRangedFloat(Minimum=-180, Maximum=180)
    Sprite Size=&Vector2DFromFloat(Value=&RandomRangeFloat(Minimum=25, Maximum=30))
   )
   &ApplyInitialForces()
   &Color()
   &DynamicMaterialParameters(
    Index 0 Param 1=1
    Index 0 Param 2=1.5
    Index 0 Param 3=&RandomRangeFloat(Minimum=0.1, Maximum=1)
    Index 0 Param 4=&RandomRangeFloat(Minimum=0.75, Maximum=1)
    Index 1 Param 1=1
    Index 1 Param 2=24
    Module.Param0WriteEnabled=true
    Module.Param0WriteEnabled1=true
    Module.Param1WriteEnabled=true
    Module.Param1WriteEnabled1=true
    Module.Param2WriteEnabled=true
    Module.Param2WriteEnabled1=true
    Module.Param3WriteEnabled=true
    Module.Param3WriteEnabled1=true
    Write Parameter Index 1=true
   )
   &CameraOffset(Camera Offset Amount=8)
  ]
  particleUpdateScript [
   &ParticleState()
   &SkeletalMeshLocation(
    Bone Sampling Mode=::NewEnumerator4
    Inherit Velocity=false
    MaskSpawnGroup=false
    Sampled Position Offset=<0 2 5>
    Sampled Position Offset Coordinate Space:ENiagaraMeshTransforms=::NewEnumerator2
    Skeletal Mesh=$User.SkeletalMesh
   )
   &SolveForcesAndVelocity(Clamp Velocity=false, Limit Acceleration=false)
   &ScaleColor(
    Scale Mode=::NewEnumerator1
    Scale RGBA=/Niagara/DynamicInputs/UniformRange/UniformRangedVector4(
     Minimum:Vector4f=&Vector4FromCurve(Vector4Curve=DataInterface() {propertyValues {
        AlphaCurve {keys [<at 0.0024 0>,<at 0.1271 1>,<at 0.265 0.9976>,<at 0.9977 -0.0026>]}
        BlueCurve {keys [<at 0.3937 0.5>,<at 0.75 0>]}
        GreenCurve {keys [<at 0.3937 2.875>,<at 0.75 0>]}
        RedCurve {keys [<at 0.3937 10>,<at 0.75 10>]}
       }})
     Maximum:Vector4f=&Vector4FromCurve(Vector4Curve=DataInterface() {propertyValues {
        AlphaCurve {keys [<at 0.0024 0>,<at 0.1271 1>,<at 0.265 0.9976>,<at 0.9977 -0.0026>]}
        BlueCurve {keys [<at 0.3937 0.5>,<at 0.75 0>]}
        GreenCurve {keys [<at 0.3937 2.875>,<at 0.75 3.7633>]}
        RedCurve {keys [<at 0.3937 10>,<at 0.75 10>]}
       }})
     Randomness Mode:ENiagaraRandomnessMode=::NewEnumerator0
     Random Seed:NiagaraInt32=0
     Fixed Random Seed:NiagaraBool=false
    )
    ScaleRGBA=true
   ) {module=ScaleColor001}
   &ScaleSpriteSize(Scale Factor=&Vector2DFromFloat(Value=&FloatFromCurve(FloatCurve=DataInterface() {propertyValues {Curve {keys [
          {arriveTangent=1.5947, interpMode=RCIM_Cubic, leaveTangent=1.5947, tangentMode=RCTM_User, time=0, value=0.5946}
          <at 0.9886 0.9389 RCIM_Cubic>
         ]}}})))
  ]
  rendererClasses [&NiagaraSpriteRendererProperties]
  renderers [{rendererClass=&NiagaraSpriteRendererProperties}]
 }
}
summary {
 systemName=NS_Backpack_Ashes
 emitters [
  {
   emitterName=Wings
   rendererClasses [
    &NiagaraSpriteRendererProperties
    &NiagaraMeshRendererProperties
    &NiagaraLightRendererProperties
   ]
  }
  {emitterName=Fire_Plume_World, rendererClasses [&NiagaraSpriteRendererProperties]}
  {
   emitterName=Tail
   rendererClasses [
    &NiagaraSpriteRendererProperties
    &NiagaraMeshRendererProperties
    &NiagaraLightRendererProperties
   ]
  }
  {bEnabled=false, emitterName=Yellow_FrontFlames, rendererClasses [&NiagaraSpriteRendererProperties]}
  {bEnabled=false, emitterName=Fire_Bottom, rendererClasses [&NiagaraSpriteRendererProperties]}
  {emitterName=decal001, rendererClasses [&NiagaraSpriteRendererProperties]}
  {emitterName=Light, rendererClasses [&NiagaraLightRendererProperties]}
  {emitterName=Fire_Bottom001, rendererClasses [&NiagaraSpriteRendererProperties]}
  {emitterName=Yellow_FrontFlames001, rendererClasses [&NiagaraSpriteRendererProperties]}
 ]
 userVariables [
  User.bIsFrontEndPreview {type=NiagaraBool}
  User.IsSkydiving {type=NiagaraBool}
  User.SkeletalMesh {
   type=NiagaraDataInterfaceSkeletalMesh
   defaultValue {
    struct=NiagaraExt_VariableValue_DataInterface
    value {dataInterface=&self:NiagaraDataInterfaceSkeletalMesh_1}
   }
  }
 ]
}
