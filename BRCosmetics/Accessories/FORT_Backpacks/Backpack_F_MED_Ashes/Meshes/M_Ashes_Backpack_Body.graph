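# Surface material for the body of the Ashes backpack cosmetic, an animated fiery glow with base color, metallic, specular and roughness fed by zero.
# Emissive adds a Fresnel orange-to-dark-red gradient to a tinted fire mask, projected in local space, blended by normal direction and partly panned.
# A time-driven sine over bounding-box UVW ripples vertices along their normals.
# Attributes pass through the shared accessory TechArt override function.
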Material /BRCosmetics/Accessories/FORT_Backpacks/Backpack_F_MED_Ashes/Meshes/M_Ashes_Backpack_Body
graph {
 properties {
  bCastRayTracedShadows=false
  bUseMaterialAttributes=true
  bUsedWithMeshDeformer=true
  bUsedWithSkeletalMesh=true
  pixelDepthOffsetMode=PDOM_Legacy
  usedShadingModels=Default Lit
 }
 nodes [
  n1 = VertexNormalWS()
  ComponentMask(None=@n1) {b=true}
  n2 = MaterialFunctionCall(Brightness=Constant() {r=1}, Exponent=Constant() {r=2}, Vertex / Pixel Normal=VertexNormalWS()) {materialFunction=&MF_Fresnel_FX}
  n3 = LinearInterpolate(A=Constant3Vector() {constant=<rgba 1 0.0648 0 1>}, B=Constant3Vector() {constant=<rgba 0.2016 0 0.0012 1>}, Alpha=@n2.Fresnel) {constA=1, constB=0}
  n4 = Divide(A=TransformPosition(None=WorldPosition().XYZ) {transformSourceType=TRANSFORMPOSSOURCE_World}, B=Multiply(A=Constant() {r=100}) {constB=-1})
  n5 = Panner(Coordinate=ComponentMask(None=@n4) {b=true, r=true}) {speedX=-0.25, speedY=0.15}
  n6 = TextureObject() {texture=/Game/Characters/Enemies/Textures/T_Elemental_Fire_M}
  n7 = Panner(Coordinate=ComponentMask(None=@n4) {b=true, r=true}) {speedX=0.25, speedY=0.15}
  n8 = Transform(None=VertexNormalWS()) {transformSourceType=TRANSFORMSOURCE_World, transformType=TRANSFORM_Local}
  n9 = Subtract(A=ComponentMask(None=@n8) {r=true})
  n10 = LinearInterpolate(
   A=LinearInterpolate(
    A=TextureSampleParameter2D(UVs=@n5, Apply View MipBias=@n6) {parameterName=Param, texture=&T_Elemental_Fire_M}.RGB
    B=TextureSample(UVs=@n7, Tex=@n6) {texture=&T_Elemental_Fire_M}.RGB
    Alpha=Saturate(None=@n9)
   )
   B=TextureSample(UVs=ComponentMask(None=@n4) {g=true, r=true}, Tex=@n6) {texture=&T_Elemental_Fire_M}.RGB
   Alpha=ComponentMask(None=@n8) {b=true}
  )
  n11 = Multiply(
   A=Multiply(A=ComponentMask(None=@n10) {b=true}, B=@n2.Fresnel Inverted)
   B=VectorParameter() {group=ElementalFire, parameterName=FireBaseColor, defaultValue=<rgba 2 0.7905 0 0>}.RGB
  )
  n12 = Multiply(A=@n1) {constB=0.12}
  n13 = OneMinus(None=Time())
  n14 = Multiply(A=@n13) {constB=2.34}
  n15 = MaterialFunctionCall() {materialFunction=&BoundingBoxBased_0-1_UVW}
  n16 = Multiply(A=@n15.B) {constB=3.1}
  n17 = Multiply(A=@n13) {constB=0.88}
  n18 = Multiply(A=@n15.B) {constB=1.3}
  n19 = Sine(None=Add(A=Add(A=@n14, B=@n16), B=Add(A=@n17, B=@n18)))
  n20 = Constant()
  n21 = SetMaterialAttributes!(
   MaterialAttributes
   Emissive Color=NamedRerouteDeclaration(None=Add(A=@n3, B=@n11)) {name=emissive, nodeColor=<rgba 0 0.7891 1 1>}
   World Position Offset=NamedRerouteDeclaration(None=Multiply(A=@n12, B=@n19)) {name=wpo, nodeColor=<rgba 0 0.5547 1 1>}
   Base Color=@n20
   Metallic=@n20
   Specular=@n20
   Roughness=@n20
  ) {attributeSetTypes [
    #1
    #2
    #3
    #4
    #5
    #6
   ]}
  MaterialFunctionCall(In=@n21) {materialFunction=&MF_TechArtOverrides_accessories}
 ]
}
